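# Material function that adds a generic power-up glow to the emissive of a Sparks instrument material, blended in by an on/off input.
# A flattened-normal Fresnel with a sine-pulsed exponent multiplies the override color, which is varied against its hue-shifted version by panning noise.
# A bounding-box gradient fades the glow toward one end.
# The electricity-distortion, chroma and flicker branch labeled "Overly detailed Powerup State" is left unconnected to the output.
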MaterialFunction /SparksCharacterCommon/Instrument/General/Materials/MF/MF_Sparks_PowerUp_GenericFX
graph {nodes [
  Constant() {r=1}
  n1 = Frac(None=Time())
  n2 = Constant() {desc=Time Mult, r=-1}
  n3 = TextureObject() {samplerType=SAMPLERTYPE_Masks, texture=/Game/Characters/Player/Male/Medium/Bodies/M_MED_Storm_Soldier/Textures/ElectricityNoise_M}
  n4 = MaterialFunctionCall!(
   In Texture=@n3
   Random Color Texture
   In Texcoords=Multiply(A=TextureCoordinate(), B=Constant2Vector() {g=2, r=3})
   Turn Time
   Distortion Tiling
   Distortion Amount=Constant() {r=0.75}
   Tiling
   Detail Factor
  ) {materialFunction=/Game/Effects/Niagara/MasterMaterials/Functions/Mojo_MF/MF_FX_TurnDistortion}
  n5 = TextureSample(UVs=@n4.Out Detail UV, Tex=@n3) {texture=/Game/Packages/Fortress_Transylvania/Haunted_Forest/Swamps/Textures/2ndInfectedMapLinear2}
  n6 = Add(
   A=Saturate(None=Multiply(A=@n1, B=@n2))
   B=Multiply(A=Power(Base=@n5.R) {constExponent=3}, B=Constant() {desc=Time Blend Noise Offset, r=0.35})
  )
  Frac(None=@n6)
  n7 = GetMaterialAttributes(None=FunctionInput() {inputName=Material, inputType=FunctionInput_MaterialAttributes}) {attributeGetTypes [
    #1
    #2
    #3
    #4
   ]}
  n8 = Constant() {r=1}
  n9 = Fresnel(
   ExponentIn=LinearInterpolate(Alpha=MaterialFunctionCall(Speed=@n8) {materialFunction=&DebugTimeSine}.Result) {constA=0.15, constB=3}
   Normal=MaterialFunctionCall(Normal=PixelNormalWS(), Flatness=ScalarParameter() {defaultValue=0.15, parameterName=FlattenNormal}) {materialFunction=&FlattenNormal}.Result
  ) {baseReflectFraction=0, exponent=1}
  n10 = MaterialFunctionCall() {materialFunction=&BoundingBoxBased_0-1_UVW}
  n11 = Power(Base=@n10.B) {constExponent=1.5}
  n12 = ScalarParameter() {defaultValue=1.8, group=Flame, parameterName=HueShift}
  n13 = FunctionInput(Preview=Constant3Vector() {constant=<rgba 0.0423 1 0 1>}) {bUsePreviewValueAsDefault=true, inputName=PowerUp_EmissiveColorOverride}
  n14 = NamedRerouteUsage()
  n15 = Multiply(
   A=TextureCoordinate()
   B=AppendVector(
    A=ScalarParameter() {defaultValue=1.25, group=Secondary, parameterName=Secondary Color TextureXScale, sortPriority=0}
    B=ScalarParameter() {defaultValue=0.3, group=Secondary, parameterName=Secondary Color TextureYScale, sortPriority=1}
   )
  )
  n16 = ScalarParameter() {group=Secondary, parameterName=Secondary Color TextureHorizPos, sortPriority=2}
  n17 = Time()
  n18 = ParticleRandom()
  n19 = Constant()
  n20 = ScalarParameter() {defaultValue=0.15, group=Secondary, parameterName=Secondary Color TextureSpeed, sortPriority=3}
  n21 = TextureSampleParameter2D(UVs=Panner(Coordinate=Add(A=@n15, B=@n16), Time=Add(A=@n17, B=@n18), Speed=AppendVector(A=@n19, B=@n20)) {speedY=0.75}) {
   group=Secondary
   parameterName=Secondary Color Texture
   samplerType=SAMPLERTYPE_LinearGrayscale
   sortPriority=8
   texture=/SparksCharacterCommon/Items/Auras/Textures/T_Mojo_Fire_Noise
  }
  n22 = Multiply(A=Power(Base=Multiply(A=@n21.G, B=@n14)) {constExponent=6}) {constB=25}
  n23 = Step(Y=@n22) {constX=0.5}
  n24 = LinearInterpolate(A=@n21.G, B=@n23, Alpha=@n23)
  n25 = Multiply(
   A=LinearInterpolate(
    A=MaterialFunctionCall(Hue Shift Percentage=@n12, Texture=@n13) {desc=Secondary color, materialFunction=&HueShift}.Result
    B=@n13
    Alpha=Saturate(None=@n24)
   )
   B=Saturate(None=Multiply(A=@n11) {constB=5})
  )
  n26 = Multiply(A=@n9, B=@n25)
  n27 = FunctionInput(Preview=Constant() {r=1}) {bUsePreviewValueAsDefault=true, inputName=Power Up Generic On/Off, inputType=FunctionInput_Scalar}
  n28 = SetMaterialAttributes!(
   MaterialAttributes=@n7.MaterialAttributes
   EmissiveColor=LinearInterpolate(A=@n7.EmissiveColor, B=Multiply(A=@n26) {constB=2, desc=Boost}, Alpha=Saturate(None=@n27))
   Roughness
   BaseColor
   WorldPositionOffset
   Opacity
  ) {attributeSetTypes [
    #3
    #2
    #1
    #4
    #5
   ]}
  FunctionOutput(None=@n28)
  n29 = TransformPosition(None=WorldPosition() {worldPositionShaderOffset=WPT_ExcludeAllShaderOffsets}.XYZ) {transformSourceType=TRANSFORMPOSSOURCE_World}
  n30 = MaterialFunctionCall() {materialFunction=&ObjectLocalBounds}
  MaterialFunctionCall(Float3=Divide(A=Subtract(A=@n29, B=@n30.Local Bounds Minimum), B=@n30.Local Bounds Size)) {materialFunction=&BreakOutFloat3Components}
  OneMinus(None=@n22)
  n31 = Saturate(None=Fresnel(ExponentIn=ScalarParameter() {defaultValue=0.45, group=Power Up Generic, parameterName=Power UpPanningEmissive_Fresnel Exponent}) {baseReflectFraction=0})
  n32 = LinearInterpolate(
   A=OneMinus(None=@n31)
   B=@n31
   Alpha=ScalarParameter() {defaultValue=1, group=Power Up Generic, parameterName=Power Up Panning Emissive_Invert Fresnel}
  )
  n33 = Add(A=Multiply(A=@n32) {constB=0.15}, B=Multiply(A=@n31, B=Sine(None=@n6)))
  n34 = MaterialFunctionCall!(In Texture=@n3, In TexCoords=@n4.Out Base UV, Offset Amount) {materialFunction=/Game/Effects/Fort_Effects/MaterialFunctions/MF_Chroma_UV}
  n35 = OneMinus(None=@n10.B)
  n36 = Saturate(None=Multiply(A=Power(Base=@n35) {constExponent=6}) {constB=5})
  n37 = LinearInterpolate(
   A=Multiply(A=@n25, B=Saturate(None=@n33))
   B=Multiply(A=Desaturation(None=@n34.Result, Fraction=@n36)) {constB=3.02}
   Alpha=Multiply(A=@n32, B=Constant() {r=0.5})
  )
  n38 = Power(Base=ComponentMask(None=@n4.Result) {r=true}) {constExponent=6}
  n39 = ScalarParameter() {defaultValue=25, parameterName=Emission Glow Detail Mult}
  n40 = Saturate(None=MaterialFunctionCall!(Time Mult=Constant() {r=2}, Time Add) {materialFunction=/Game/Effects/Niagara/MasterMaterials/Functions/Mojo_MF/MF_FX_Flicker}.Raw Flicker)
  n41 = Multiply(A=@n37, B=Multiply(A=Multiply(A=@n38, B=@n39) {constB=15}, B=LinearInterpolate(Alpha=@n40) {constA=0.25, constB=2})) {constB=25}
  n42 = Multiply(A=@n37, B=NamedRerouteDeclaration(None=OneMinus(None=@n36)) {name=LocalSpaceMaskV, nodeColor=<rgba 0 0.5312 1 1>})
  n43 = Saturate(None=Power(Base=@n31))
  Reroute(None=LinearInterpolate(A=Add(A=@n41, B=@n42), B=Multiply(A=@n25, B=@n43), Alpha=@n43)) {desc=Overly detailed Powerup State}
  n44 = ShadingPathSwitch(Default=@n9, Deferred=@n9, Forward=VectorParameter() {parameterName=Forward, defaultValue=<rgba 0 0 1 1>}.RGB, Mobile=@n9)
  Reroute(None=ShadingPathSwitch(Default=@n44, Deferred=@n44, Mobile=Constant() {r=0.05}) {desc=Migrated from FeatureLevelSwitch})
 ]}
